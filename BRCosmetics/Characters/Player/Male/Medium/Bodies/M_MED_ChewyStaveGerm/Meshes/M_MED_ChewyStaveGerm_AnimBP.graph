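# Post-copy AnimBP for the ChewyStaveGerm male medium cosmetic body, adding secondary motion to pant, chain and collar bones.
# It copies the parent mesh pose, runs a Control Rig fed emote and skydive floats, and caches the result.
# Chained Anim Dynamics springs follow, with a weighted spine constraint positioning the chain base.
# Update also computes gravity, RBAN and front-end simulation settings, which the shown AnimGraph never reads.
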AnimBlueprint /BRCosmetics/Characters/Player/Male/Medium/Bodies/M_MED_ChewyStaveGerm/Meshes/M_MED_ChewyStaveGerm_AnimBP
target_skeleton=&/M_MED_ChewyStaveGerm_Skeleton
graphs [
 AnimGraph {nodes [
   n1 = Anim Dynamics - Bone: dyn_pant_lwr_r_01!(
    ComponentPose=Local To Component(LocalPose=Use cached pose 'Pre_ControlRig'())
    Alpha=1
    GravityOverride:Vector="0,0,-980"
    ExternalForce="0,0,0"
   ) -> (Pose) {
    angularDampingOverride=0.999
    angularSpringConstant=2000
    bAngularSpring=true
    bOverrideAngularDamping=true
    bUseGravityOverride=true
    gravityScale=0.35
    simulationSpace=RootRelative
    boundBone {boneName=dyn_pant_lwr_r_01}
    componentAppliedLinearAccClamp=<50000 50000 50000>
    componentLinearAccScale=<1 1 1>
    componentLinearVelScale=<1 1 1>
    gravityOverride {z=-980}
    physicsBodyDefinitions [{
      boundBone {boneName=dyn_pant_lwr_r_01}
      boxExtents=<15 20 20>
      constraintSetup {angularLimitsMax=<0 30 30>, angularLimitsMin=<0 -30 -30>, angularTarget=<1 0 0>}
      localJointOffset=<4 0 0>
     }]
   }
   n2 = Anim Dynamics - Bone: dyn_pant_lwr_l_01!(ComponentPose=@n1, Alpha=1, GravityOverride:Vector="0,0,-980", ExternalForce="0,0,0") -> (Pose) {
    angularDampingOverride=0.999
    angularSpringConstant=2000
    bAngularSpring=true
    bOverrideAngularDamping=true
    bUseGravityOverride=true
    gravityScale=0.35
    simulationSpace=RootRelative
    boundBone {boneName=dyn_pant_lwr_l_01}
    componentAppliedLinearAccClamp=<50000 50000 50000>
    componentLinearAccScale=<1 1 1>
    componentLinearVelScale=<1 1 1>
    gravityOverride {z=-980}
    physicsBodyDefinitions [{
      boundBone {boneName=dyn_pant_lwr_l_01}
      boxExtents=<15 20 20>
      constraintSetup {angularLimitsMax=<0 30 30>, angularLimitsMin=<0 -30 -30>, angularTarget=<1 0 0>}
      localJointOffset=<4 0 0>
     }]
   }
   n3 = "Constraint - base_chain_01 (PPPP)"!(
    ComponentPose=@n2
    Alpha=1
    ConstraintWeights_0=0.5406
    ConstraintWeights_1:float=0.2316
    ConstraintWeights_2:float=0.2061
    ConstraintWeights_3:float=0.0217
   ) -> (Pose) {
    boneToModify {boneName=base_chain_01}
    constraintSetup [{targetBone {boneName=spine_04}}, {targetBone {boneName=spine_03}}, {targetBone {boneName=spine_05}}, {targetBone {boneName=spine_02}}]
    constraintWeights [1,1,1,1]
   }
   n4 = Anim Dynamics - Bone: dyn_chain_01!(ComponentPose=@n3, Alpha=1, GravityOverride:Vector="0,0,-980", ExternalForce="0,0,0") -> (Pose) {
    angularDampingOverride=0.9999
    angularSpringConstant=1498
    bAngularSpring=true
    bOverrideAngularDamping=true
    bUseGravityOverride=true
    gravityScale=0.35
    simulationSpace=RootRelative
    boundBone {boneName=dyn_chain_01}
    componentAppliedLinearAccClamp=<2500 2500 2500>
    componentLinearAccScale=<1 1 1>
    componentLinearVelScale=<1 1 1>
    gravityOverride {z=-980}
    physicsBodyDefinitions [{
      boundBone {boneName=dyn_chain_01}
      boxExtents=<10 10 10>
      constraintSetup {angularLimitsMax=<0 -10 20>, angularLimitsMin=<0 -50 -20>, angularTarget=<1 0 0>}
      localJointOffset=<2 0 0>
     }]
    simSpaceSettings {simSpaceAngularAlpha=0.3}
   }
   n5 = Anim Dynamics - Bone: dyn_collar_f_l_01!(
    SimulationSpace:AnimPhysSimSpaceType Enum=BoneRelative
    ComponentPose=@n4
    Alpha=1
    GravityOverride:Vector="0,0,-980"
    AngularSpringConstant:float=6000
    ExternalForce="0,0,0"
   ) -> (Pose) {
    angularDampingOverride=0.99
    angularSpringConstant=6000
    bAngularSpring=true
    bOverrideAngularDamping=true
    bUseGravityOverride=true
    gravityScale=0.35
    simulationSpace=BoneRelative
    alphaBoolBlend {blendOption=Linear}
    boundBone {boneName=dyn_collar_f_l_01}
    componentAppliedLinearAccClamp=<10000 10000 10000>
    componentLinearAccScale=<1 1 1>
    componentLinearVelScale=<1 1 1>
    gravityOverride {z=-980}
    physicsBodyDefinitions [{
      boundBone {boneName=dyn_collar_f_l_01}
      boxExtents=<20 30 30>
      constraintSetup {angularLimitsMax=<0 10 0>, angularLimitsMin=<0 -10 0>, angularTarget=<1 0 0>}
      localJointOffset=<2 0 0>
     }]
    relativeSpaceBone {boneName=root}
    simSpaceSettings {simSpaceAngularAlpha=0.5}
   }
   n6 = Anim Dynamics - Bone: dyn_collar_f_r_01!(
    SimulationSpace:AnimPhysSimSpaceType Enum=BoneRelative
    ComponentPose=@n5
    Alpha=1
    GravityOverride:Vector="0,0,-980"
    AngularSpringConstant:float=6000
    ExternalForce="0,0,0"
   ) -> (Pose) {
    angularDampingOverride=0.99
    angularSpringConstant=6000
    bAngularSpring=true
    bOverrideAngularDamping=true
    bUseGravityOverride=true
    gravityScale=0.35
    simulationSpace=BoneRelative
    alphaBoolBlend {blendOption=Linear}
    boundBone {boneName=dyn_collar_f_r_01}
    componentAppliedLinearAccClamp=<10000 10000 10000>
    componentLinearAccScale=<1 1 1>
    componentLinearVelScale=<1 1 1>
    gravityOverride {z=-980}
    physicsBodyDefinitions [{
      boundBone {boneName=dyn_collar_f_r_01}
      boxExtents=<20 30 30>
      constraintSetup {angularLimitsMax=<0 10 0>, angularLimitsMin=<0 -10 0>, angularTarget=<1 0 0>}
      localJointOffset=<2 0 0>
     }]
    relativeSpaceBone {boneName=root}
    simSpaceSettings {simSpaceAngularAlpha=0.5}
   }
   Output Pose(Result=Component To Local(ComponentPose=@n6)) {layerGroup=None}
   n7 = Control Rig!(
    Source=Copy Pose From Mesh() {bUseAttachedParent=true}
    Alpha
    isPlayingEmoteFloat:double=Get isEmotingFloat() -> (isEmotingFloat:double)
    isSkyDivingFloat:double=Get isSkyDivingFloat() -> (isSkyDivingFloat:double)
   ) -> (Pose) {controlRigAssetReference {blueprintRigClass=&/M_MED_ChewyStaveGerm_PRE_CtrlRig.M_MED_ChewyStaveGerm_PRE_CtrlRig_C}}
   Save cached pose 'Pre_ControlRig'(Pose=@n7) {cacheName=Pre_ControlRig}
  ]}
 EventGraph {nodes [
   Event BlueprintUpdateAnimation()
   > n1 = Sequence!(execute) -> (then_0, then_1, then_2:Exec)
   > n2 = "Gravity Override State Function (Parameter Asset) (Nativized)"(
    CharacterPartAnimInstance=Self-Reference()
    GameTimeInSeconds=GetGameTimeInSeconds()
    Parameters=&/F_MED_ChewyStaveGerm_Parameters
   )
   > Set Override World Gravity(Override World Gravity:Vector=@n2) -> (Output_Get:Vector)
   > Set RBAN Settings Out(RBAN Settings Out=@n2.OutRigidBodyAnimNodeParameters)
   Set isEmotingFloat(execute=@n1.then_1, isEmotingFloat:double="To Float (Boolean)"(InBool=Get bIsPlayingEmote())) -> (Output_Get:double)
   > Set isSkyDivingFloat(isSkyDivingFloat:double="To Float (Boolean)"(InBool=Get bIsSkyDiving())) -> (Output_Get:double)
   n3 = Branch(execute=@n1.then_2, Condition=Get bIsInFrontEnd())
   > Set SimlationSpace(SimlationSpace:AnimPhysSimSpaceType Enum=World) -> (Output_Get:AnimPhysSimSpaceType Enum)
   > Set AlphaFE(AlphaFE:double=0.3) -> (Output_Get:double)
   Set SimlationSpace(execute=@n3.else)
   > Set AlphaFE(AlphaFE=1)
  ]}
]
variables [
 AlphaFE {type=number}
 isEmotingFloat {type=number}
 isSkyDivingFloat {type=number}
 Override World Gravity {type=Vector}
 RBAN Settings Out {type=FortRigidBodyAnimNodeParameters}
 SimlationSpace {type=AnimPhysSimSpaceType}
]
