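# Physics asset for the M_MED_ChewyStaveGerm male player body cosmetic, covering its collar and upper torso.
# Four two-bone collar chains (front and side, left and right) link tiny base capsules to spine_05 with free angular motion.
# Each larger dyn collar capsule hangs off its base with near-zero swing and twist limits.
# Spine, clavicle and upper arm capsules supply body volumes, with the clavicles and arms left unconstrained.
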PhysicsAsset /BRCosmetics/Characters/Player/Male/Medium/Bodies/M_MED_ChewyStaveGerm/Meshes/M_MED_ChewyStaveGerm_Physics
bodies {
 base_collar_f_l_01 [{shapeType=Capsule, length=0.505, radius=0.505, shapeName=base_collar_f_l_01_capsule, rotation=<rot 0 -90 0>}]
 base_collar_f_r_01 [{shapeType=Capsule, length=0.505, radius=0.505, shapeName=base_collar_f_r_01_capsule, rotation=<rot 0 -90 0>}]
 base_collar_l_01 [{shapeType=Capsule, length=0.505, radius=0.505, shapeName=base_collar_l_01_capsule, rotation=<rot 0 -90 0>}]
 base_collar_r_01 [{shapeType=Capsule, length=0.505, radius=0.505, shapeName=base_collar_r_01_capsule, rotation=<rot 0 -90 0>}]
 clavicle_l [{
   shapeType=Capsule
   length=11.5639
   radius=8.4556
   shapeName=clavicle_l_capsule
   center=<9.6982 -0.8068 -1.0553>
   rotation=<rot -14.7542 95.0564 -106.8843>
  }]
 clavicle_r [{
   shapeType=Capsule
   length=11.5639
   radius=8.4556
   shapeName=clavicle_r_capsule
   center=<-9.6982 0.8068 1.0553>
   rotation=<rot -14.7542 95.0564 73.1157>
  }]
 dyn_collar_f_l_01 [{
   shapeType=Capsule
   length=6.9843
   radius=3.9647
   shapeName=dyn_collar_f_l_01_capsule
   center=<3.9664 0.9636 -2.2144>
   rotation=<rot -4.1671 -4.6141 -79.7877>
  }]
 dyn_collar_f_r_01 [{
   shapeType=Capsule
   length=8.2601
   radius=2.8588
   shapeName=dyn_collar_f_r_01_capsule
   center=<-4.8604 -1.8761 0.8272>
   rotation=<rot -13.5949 178.4598 72.1753>
  }]
 dyn_collar_l_01 [{
   shapeType=Capsule
   length=7.1937
   radius=2.1687
   shapeName=dyn_collar_l_01_capsule
   center=<2.3 2.1 -2.5>
   rotation=<rot 7.0152 9.4509 -64.3389>
  }]
 dyn_collar_r_01 [{
   shapeType=Capsule
   length=7.1937
   radius=2.1687
   shapeName=dyn_collar_r_01_capsule
   center=<-2.3 -2.1 2.5>
   rotation=<rot 7.0152 9.4509 115.6611>
  }]
 spine_05 [{shapeType=Capsule, length=14.7941, radius=12.2077, shapeName=spine_05_capsule, center=<-1.1188 -0.7736 0>, rotation=<rot 0 -90 0>}]
 upperarm_l [{
   shapeType=Capsule
   length=21.1101
   radius=6.8246
   shapeName=upperarm_l_capsule
   center=<10.0919 -0.5726 -1.6325>
   rotation=<rot 1.091 -87.6365 -91.4447>
  }]
 upperarm_r [{
   shapeType=Capsule
   length=21.1101
   radius=6.8246
   shapeName=upperarm_r_capsule
   center=<-10.0919 0.5726 1.6325>
   rotation=<rot 1.091 -87.6365 88.5553>
  }]
}
constraints [
 {
  bone1Name=base_collar_f_r_01
  bone2Name=spine_05
  swing1LimitDegrees=45
  swing1Motion=Free
  swing2LimitDegrees=45
  swing2Motion=Free
  twistLimitDegrees=45
  twistMotion=Free
 }
 {bone1Name=dyn_collar_f_r_01, bone2Name=base_collar_f_r_01, swing1LimitDegrees=1, swing2LimitDegrees=1, twistLimitDegrees=0}
 {
  bone1Name=base_collar_l_01
  bone2Name=spine_05
  swing1LimitDegrees=45
  swing1Motion=Free
  swing2LimitDegrees=45
  swing2Motion=Free
  twistLimitDegrees=45
  twistMotion=Free
 }
 {bone1Name=dyn_collar_l_01, bone2Name=base_collar_l_01, swing1LimitDegrees=1, swing2LimitDegrees=1, twistLimitDegrees=0}
 {
  bone1Name=base_collar_r_01
  bone2Name=spine_05
  swing1LimitDegrees=45
  swing1Motion=Free
  swing2LimitDegrees=45
  swing2Motion=Free
  twistLimitDegrees=45
  twistMotion=Free
 }
 {bone1Name=dyn_collar_r_01, bone2Name=base_collar_r_01, swing1LimitDegrees=1, swing2LimitDegrees=1, twistLimitDegrees=0}
 {
  bone1Name=base_collar_f_l_01
  bone2Name=spine_05
  swing1LimitDegrees=45
  swing1Motion=Free
  swing2LimitDegrees=45
  swing2Motion=Free
  twistLimitDegrees=45
  twistMotion=Free
 }
 {bone1Name=dyn_collar_f_l_01, bone2Name=base_collar_f_l_01, swing1LimitDegrees=1, swing2LimitDegrees=1, twistLimitDegrees=0}
]
