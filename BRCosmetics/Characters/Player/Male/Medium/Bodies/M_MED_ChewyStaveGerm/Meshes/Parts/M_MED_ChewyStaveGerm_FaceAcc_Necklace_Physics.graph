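# Physics asset for the necklace face accessory on the ChewyStaveGerm medium male player body, defining its bodies and joints.
# A three-bone dyn_necklace chain of capsules hangs from a small base_necklace_01 body, which links through spine_05 to spine_04.
# Larger spine capsules approximate the upper torso, with two shapes on spine_05.
# Spine and base joints rotate freely, while chain joints have tight twist limits and swing limits that are widest at the last link.
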PhysicsAsset /BRCosmetics/Characters/Player/Male/Medium/Bodies/M_MED_ChewyStaveGerm/Meshes/Parts/M_MED_ChewyStaveGerm_FaceAcc_Necklace_Physics
bodies {
 base_necklace_01 [{shapeType=Capsule, length=0.505, radius=0.505, shapeName=base_necklace_01_capsule, rotation=<rot 0 -90 0>}]
 dyn_necklace_01 [{shapeType=Capsule, length=13.6975, radius=2, shapeName=dyn_necklace_01_capsule, center=<4 0 1.2>, rotation=<rot -90 0 90>}]
 dyn_necklace_02 [{shapeType=Capsule, length=6.7401, radius=2, shapeName=dyn_necklace_02_capsule, center=<4 0 -1.2>, rotation=<rot -90 0 90>}]
 dyn_necklace_03 [{shapeType=Capsule, length=2, radius=2, shapeName=dyn_necklace_03_capsule, center=<2 0 -1>, rotation=<rot -90 0 0>}]
 spine_04 [{shapeType=Capsule, length=15.87, radius=11.5, shapeName=spine_04_capsule, center=<4.5303 1.1113 0>, rotation=<rot 0 -90 0>}]
 spine_05 [
  {shapeType=Capsule, length=14.4923, radius=7.3239, center=<1.7631 -0.3865 0.3535>, rotation=<rot -0.1692 -131.3552 -86.4553>}
  {shapeType=Capsule, length=15, radius=15, shapeName=spine_05_capsule, center=<-4.1937 -4.1585 0>, rotation=<rot 0 -19.3072 0>}
 ]
}
constraints [
 {
  bone1Name=spine_05
  bone2Name=spine_04
  swing1LimitDegrees=45
  swing1Motion=Free
  swing2LimitDegrees=45
  swing2Motion=Free
  twistLimitDegrees=45
  twistMotion=Free
 }
 {
  bone1Name=base_necklace_01
  bone2Name=spine_05
  swing1LimitDegrees=45
  swing1Motion=Free
  swing2LimitDegrees=45
  swing2Motion=Free
  twistLimitDegrees=45
  twistMotion=Free
 }
 {bone1Name=dyn_necklace_01, bone2Name=base_necklace_01, swing1LimitDegrees=10, swing2LimitDegrees=10, twistLimitDegrees=1}
 {bone1Name=dyn_necklace_02, bone2Name=dyn_necklace_01, swing1LimitDegrees=5, swing2LimitDegrees=10, twistLimitDegrees=1}
 {bone1Name=dyn_necklace_03, bone2Name=dyn_necklace_02, swing1LimitDegrees=30, swing2LimitDegrees=60, twistLimitDegrees=1}
]
