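# Post-process style AnimBP for the ChewyStaveGerm male face accessory, adding secondary motion to hair and necklace.
# The parent-mesh pose feeds a movement-state-driven Control Rig, necklace RigidBody simulation, hair Look Ats, six spring AnimDynamics nodes and an additive necklace bone rotation.
# The update event converts state bools to floats and computes gravity and RigidBody settings from a parameter asset.
# In the front end, outside emotes and skydiving, hair simulates in world space at reduced alpha.
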AnimBlueprint /BRCosmetics/Characters/Player/Male/Medium/Bodies/M_MED_ChewyStaveGerm/Meshes/Parts/M_MED_ChewyStaveGerm_FaceAcc_AnimBP
target_skeleton=&/M_MED_ChewyStaveGerm_FaceAcc_Skeleton
graphs [
 AnimGraph {nodes [
   n1 = Get Anim_Dynamics_Alpha_Frontend() -> (Anim_Dynamics_Alpha_Frontend:double)
   n2 = Get AnimDynamics_SimSpace_Frontend() -> (AnimDynamics_SimSpace_Frontend:AnimPhysSimSpaceType Enum)
   n3 = Control Rig!(
    Source=Copy Pose From Mesh() {bUseAttachedParent=true}
    Alpha
    isCruchingFloat:double=Get isCrouchingFloat() -> (isCrouchingFloat:double)
    isTargetingFloat:double=Get isTargetingFloat() -> (isTargetingFloat:double)
    isTacticalSprintingFloat:double=Get isTacticalSprintingFloat() -> (isTacticalSprintingFloat:double)
    isSkydivingFloat:double=Get IsSkyDivingFloat() -> (IsSkyDivingFloat:double)
    isJumpingFloat:double=Get isjumpingFloat() -> (isjumpingFloat:double)
    isPlayingEmote:double=Get isEmotingFloat() -> (isEmotingFloat:double)
    isInAir:double=Get isInAirFloat() -> (isInAirFloat:double)
   ) -> (Pose) {controlRigAssetReference {blueprintRigClass=&/M_MED_ChewyStaveGerm_FaceAcc_CtrlRig.M_MED_ChewyStaveGerm_FaceAcc_CtrlRig_C}}
   n4 = RigidBody!(
    ComponentPose=Local To Component(LocalPose=@n3)
    Alpha
    OverrideWorldGravity:Vector="0,0,0"
    ExternalForce
    SimSpaceSettings:Sim Space Settings Structure=(WorldAlpha=0,VelocityScaleZ=1,DampingAlpha=1,MaxLinearVelocity=10000,MaxAngularVelocity=10000,MaxLinearAcceleration=10000,MaxAngularAcceleration=10000,ExternalLinearDragV=(X=0,Y=0,Z=0),ExternalLinearVelocity=(X=0,Y=0,Z=0),ExternalAngularVelocity=(X=0,Y=0,Z=0))
   ) -> (Pose) {
    bOverrideWorldGravity=true
    overridePhysicsAsset=&/M_MED_ChewyStaveGerm_FaceAcc_Necklace_Physics
    simulationSpace=BaseBoneSpace
    baseBoneRef {boneName=spine_03}
   }
   n5 = Look At - Bone: base_Hair_01_l!(ComponentPose=@n4, Alpha=1, LookAtLocation:Vector="3,-12,-3.25") -> (Pose) {
    boneToModify {boneName=base_Hair_01_l}
    lookAtLocation=<3 -12 -3.25>
    lookAtTarget {boneReference {boneName=spine_05}}
    lookAt_Axis {axis=<1 0>}
   }
   n6 = Look At - Bone: base_Hair_01_r(ComponentPose=@n5, Alpha=1) {boneToModify {boneName=base_Hair_01_r}, lookAtLocation=<3 -12 8>, lookAtTarget {boneReference {boneName=spine_05}}, lookAt_Axis {axis=<1 0>}}
   n7 = Look At - Bone: base_Hair_01_m(ComponentPose=@n6, Alpha=1) {boneToModify {boneName=base_Hair_01_m}, lookAtLocation=<3 -15>, lookAtTarget {boneReference {boneName=spine_05}}, lookAt_Axis {axis=<1 0>}}
   n8 = Anim Dynamics - Bone: dyn_Hair_01_l!(SimulationSpace:AnimPhysSimSpaceType Enum=@n2, ComponentPose=@n7, Alpha=@n1, GravityOverride:Vector="0,0,0", ExternalForce="0,0,0") -> (Pose) {
    angularBiasOverride=0.9999
    angularDampingOverride=0.9999
    bLinearSpring=true
    bOverrideAngularBias=true
    bOverrideAngularDamping=true
    bOverrideLinearDamping=true
    bUseGravityOverride=true
    linearDampingOverride=0.9999
    linearSpringConstant=100
    simulationSpace=RootRelative
    alphaBoolBlend {blendOption=Linear}
    alphaScaleBiasClamp {scale=0.5}
    boundBone {boneName=dyn_Hair_01_l}
    componentAppliedLinearAccClamp=<1250 1250 1250>
    componentLinearAccScale=<0.4 0.5 0.5>
    componentLinearVelScale=<0.4 0.5 0.5>
    physicsBodyDefinitions [{
      boundBone {boneName=dyn_Hair_01_l}
      boxExtents=<12 8 8>
      constraintSetup {angularTargetAxis=AxisZ, angularLimitsMax=<10 15 20>, angularLimitsMin=<-10 -15 -30>, angularTarget=<1 1 1>}
      localJointOffset=<4 0 0>
     }]
    simSpaceSettings {simSpaceAngularAlpha=0.5}
   }
   n9 = Anim Dynamics - Bone: dyn_Hair_01_m!(SimulationSpace:AnimPhysSimSpaceType Enum=@n2, ComponentPose=@n8, Alpha=@n1, GravityOverride:Vector="0,0,0", ExternalForce="0,0,0") -> (Pose) {
    angularBiasOverride=0.9999
    angularDampingOverride=0.9999
    bLinearSpring=true
    bOverrideAngularBias=true
    bOverrideAngularDamping=true
    bOverrideLinearDamping=true
    bUseGravityOverride=true
    linearDampingOverride=0.9999
    linearSpringConstant=100
    simulationSpace=RootRelative
    alphaBoolBlend {blendOption=Linear}
    alphaScaleBiasClamp {scale=0.5}
    boundBone {boneName=dyn_Hair_01_m}
    componentAppliedLinearAccClamp=<1250 1250 1250>
    componentLinearAccScale=<0.4 0.5 0.5>
    componentLinearVelScale=<0.4 0.5 0.5>
    physicsBodyDefinitions [{
      boundBone {boneName=dyn_Hair_01_m}
      boxExtents=<12 8 8>
      constraintSetup {angularTargetAxis=AxisZ, angularLimitsMax=<10 15 20>, angularLimitsMin=<-10 -15 -30>, angularTarget=<1 1 1>}
      localJointOffset=<4 0 0>
     }]
    simSpaceSettings {simSpaceAngularAlpha=0.5}
   }
   n10 = Anim Dynamics - Bone: dyn_Hair_01_r!(SimulationSpace:AnimPhysSimSpaceType Enum=@n2, ComponentPose=@n9, Alpha=@n1, GravityOverride:Vector="0,0,0", ExternalForce="0,0,0") -> (Pose) {
    angularBiasOverride=0.9999
    angularDampingOverride=0.9999
    bLinearSpring=true
    bOverrideAngularBias=true
    bOverrideAngularDamping=true
    bOverrideLinearDamping=true
    bUseGravityOverride=true
    linearDampingOverride=0.9999
    linearSpringConstant=100
    simulationSpace=RootRelative
    alphaBoolBlend {blendOption=Linear}
    alphaScaleBiasClamp {scale=0.5}
    boundBone {boneName=dyn_Hair_01_r}
    componentAppliedLinearAccClamp=<1250 1250 1250>
    componentLinearAccScale=<0.4 0.5 0.5>
    componentLinearVelScale=<0.4 0.5 0.5>
    physicsBodyDefinitions [{
      boundBone {boneName=dyn_Hair_01_r}
      boxExtents=<12 8 8>
      constraintSetup {angularTargetAxis=AxisZ, angularLimitsMax=<10 15 20>, angularLimitsMin=<-10 -15 -30>, angularTarget=<1 1 1>}
      localJointOffset=<4 0 0>
     }]
    simSpaceSettings {simSpaceAngularAlpha=0.5}
   }
   n11 = Anim Dynamics - Bone: dyn_Hair_02_l!(SimulationSpace:AnimPhysSimSpaceType Enum=@n2, ComponentPose=@n10, Alpha=@n1, GravityOverride:Vector="0,0,0", ExternalForce="0,0,0") -> (Pose) {
    angularBiasOverride=0.9999
    angularDampingOverride=0.9999
    bLinearSpring=true
    bOverrideAngularBias=true
    bOverrideAngularDamping=true
    bOverrideLinearDamping=true
    bUseGravityOverride=true
    linearDampingOverride=0.9999
    linearSpringConstant=100
    simulationSpace=RootRelative
    alphaBoolBlend {blendOption=Linear}
    alphaScaleBiasClamp {scale=0.5}
    boundBone {boneName=dyn_Hair_02_l}
    componentAppliedLinearAccClamp=<1250 1250 1250>
    componentLinearAccScale=<0.4 0.5 0.5>
    componentLinearVelScale=<0.4 0.5 0.5>
    physicsBodyDefinitions [{
      boundBone {boneName=dyn_Hair_02_l}
      boxExtents=<12 8 8>
      constraintSetup {angularTargetAxis=AxisZ, angularLimitsMax=<10 15 20>, angularLimitsMin=<-10 -15 -30>, angularTarget=<1 1 1>}
      localJointOffset=<4 0 0>
     }]
    simSpaceSettings {simSpaceAngularAlpha=0.5}
   }
   n12 = Anim Dynamics - Bone: dyn_Hair_02_m!(SimulationSpace:AnimPhysSimSpaceType Enum=@n2, ComponentPose=@n11, Alpha=@n1, GravityOverride:Vector="0,0,0", ExternalForce="0,0,0") -> (Pose) {
    angularBiasOverride=0.9999
    angularDampingOverride=0.9999
    bLinearSpring=true
    bOverrideAngularBias=true
    bOverrideAngularDamping=true
    bOverrideLinearDamping=true
    bUseGravityOverride=true
    linearDampingOverride=0.9999
    linearSpringConstant=100
    simulationSpace=RootRelative
    alphaBoolBlend {blendOption=Linear}
    alphaScaleBiasClamp {scale=0.5}
    boundBone {boneName=dyn_Hair_02_m}
    componentAppliedLinearAccClamp=<1250 1250 1250>
    componentLinearAccScale=<0.4 0.5 0.5>
    componentLinearVelScale=<0.4 0.5 0.5>
    physicsBodyDefinitions [{
      boundBone {boneName=dyn_Hair_02_m}
      boxExtents=<12 8 8>
      constraintSetup {angularTargetAxis=AxisZ, angularLimitsMax=<10 15 20>, angularLimitsMin=<-10 -15 -30>, angularTarget=<1 1 1>}
      localJointOffset=<4 0 0>
     }]
    simSpaceSettings {simSpaceAngularAlpha=0.5}
   }
   n13 = Anim Dynamics - Bone: dyn_Hair_02_r!(SimulationSpace:AnimPhysSimSpaceType Enum=@n2, ComponentPose=@n12, Alpha=@n1, GravityOverride:Vector="0,0,0", ExternalForce="0,0,0") -> (Pose) {
    angularBiasOverride=0.9999
    angularDampingOverride=0.9999
    bLinearSpring=true
    bOverrideAngularBias=true
    bOverrideAngularDamping=true
    bOverrideLinearDamping=true
    bUseGravityOverride=true
    linearDampingOverride=0.9999
    linearSpringConstant=100
    simulationSpace=RootRelative
    alphaBoolBlend {blendOption=Linear}
    alphaScaleBiasClamp {scale=0.5}
    boundBone {boneName=dyn_Hair_02_r}
    componentAppliedLinearAccClamp=<1250 1250 1250>
    componentLinearAccScale=<0.4 0.5 0.5>
    componentLinearVelScale=<0.4 0.5 0.5>
    physicsBodyDefinitions [{
      boundBone {boneName=dyn_Hair_02_r}
      boxExtents=<12 8 8>
      constraintSetup {angularTargetAxis=AxisZ, angularLimitsMax=<10 15 20>, angularLimitsMin=<-10 -15 -30>, angularTarget=<1 1 1>}
      localJointOffset=<4 0 0>
     }]
    simSpaceSettings {simSpaceAngularAlpha=0.5}
   }
   n14 = "Transform (Modify) Bone - Bone: dyn_necklace_03"!(ComponentPose=@n13, bAlphaBoolEnabled:bool=True, Translation="0,0,0", Rotation="10,0,0", Scale="1,1,1") -> (Pose) {
    alphaInputType=Bool
    rotationMode=BMM_Additive
    rotationSpace=BCS_BoneSpace
    translationMode=BMM_Additive
    translationSpace=BCS_BoneSpace
    boneToModify {boneName=dyn_necklace_03}
    rotation {pitch=10}
   }
   Output Pose(Result=Component To Local(ComponentPose=@n14)) {layerGroup=None}
  ]}
 EventGraph {nodes [
   Event BlueprintUpdateAnimation()
   > n1 = Sequence!(execute) -> (then_0, then_1, then_2:Exec, then_3:Exec, then_4:Exec)
   > n2 = "Gravity Override State Function (Parameter Asset) (Nativized)"(
    CharacterPartAnimInstance=Self-Reference()
    GameTimeInSeconds=GetGameTimeInSeconds()
    Parameters=&/F_MED_ChewyStaveGerm_FaceACC_Parameters
   )
   > Set Override World Gravity(Override World Gravity:Vector=@n2) -> (Output_Get:Vector)
   > Set RBAN Settings Out(RBAN Settings Out=@n2.OutRigidBodyAnimNodeParameters)
   Set isCrouchingFloat(execute=@n1.then_1, isCrouchingFloat:double="To Float (Boolean)"(InBool=Get bIsCrouching())) -> (Output_Get:double)
   > Set isTargetingFloat(isTargetingFloat:double="To Float (Boolean)"(InBool=Get bIsTargeting())) -> (Output_Get:double)
   > Set isTacticalSprintingFloat(isTacticalSprintingFloat:double="To Float (Boolean)"(InBool=Get bIsTacticalSprinting())) -> (Output_Get:double)
   > Set IsSkyDivingFloat(IsSkyDivingFloat:double="To Float (Boolean)"(InBool=Get bIsSkyDiving())) -> (Output_Get:double)
   n3 = NOT Boolean(A=Get bIsSkyDiving())
   n4 = Get bIsJumping()
   > Set isjumpingFloat(isjumpingFloat:double="To Float (Boolean)"(InBool=AND Boolean(A=@n4, B=@n3))) -> (Output_Get:double)
   > Set isEmotingFloat(isEmotingFloat:double="To Float (Boolean)"(InBool=Get bIsPlayingEmote())) -> (Output_Get:double)
   > Set isInAirFloat(isInAirFloat:double="To Float (Boolean)"(InBool=Get bIsInAir() -> (bIsInAir:bool))) -> (Output_Get:double)
   n5 = NOT Boolean(A=Get bIsPlayingEmote())
   n6 = NOT Boolean(A=Get bIsSkyDiving())
   n7 = Branch(execute=@n1.then_2, Condition=AND Boolean(A=Get bIsInFrontEnd(), B=AND Boolean(A=@n6, B=@n5)))
   > Set AnimDynamics_SimSpace_Frontend(AnimDynamics_SimSpace_Frontend:AnimPhysSimSpaceType Enum=World) -> (Output_Get:AnimPhysSimSpaceType Enum)
   > Set Anim_Dynamics_Alpha_Frontend(Anim_Dynamics_Alpha_Frontend:double=0.4) -> (Output_Get:double)
   Set AnimDynamics_SimSpace_Frontend(execute=@n7.else)
   > Set Anim_Dynamics_Alpha_Frontend(Anim_Dynamics_Alpha_Frontend=1)
  ]}
]
variables [
 Anim_Dynamics_Alpha_Frontend {type=number}
 AnimDynamics_SimSpace_Frontend {type=AnimPhysSimSpaceType}
 isCrouchingFloat {type=number}
 isEmotingFloat {type=number}
 isInAirFloat {type=number}
 isjumpingFloat {type=number}
 IsSkyDivingFloat {type=number}
 isTacticalSprintingFloat {type=number}
 isTargetingFloat {type=number}
 LookAtLocationLHair {type=Vector}
 Override World Gravity {type=Vector}
 RBAN Settings Out {type=FortRigidBodyAnimNodeParameters}
]
