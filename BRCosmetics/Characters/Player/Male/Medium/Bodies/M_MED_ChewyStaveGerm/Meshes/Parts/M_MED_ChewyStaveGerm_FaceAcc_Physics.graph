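# Physics asset for the FaceAcc part of the male medium ChewyStaveGerm character, defining capsule bodies and constraints for three hair strands (left, middle, right).
# Each strand chains a small base body to the head, then two larger dyn bodies.
# Angular limits are tight: near-zero at the head, zero at the first dyn link, slight swing without twist at the tip.
# Head, neck and upper-spine capsules add body volumes around the hair.
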PhysicsAsset /BRCosmetics/Characters/Player/Male/Medium/Bodies/M_MED_ChewyStaveGerm/Meshes/Parts/M_MED_ChewyStaveGerm_FaceAcc_Physics
bodies {
 base_Hair_01_l [{shapeType=Capsule, length=0.505, radius=0.505, shapeName=base_Hair_01_l_capsule, rotation=<rot 0 -90 0>}]
 base_Hair_01_m [{shapeType=Capsule, length=0.505, radius=0.505, shapeName=base_Hair_01_m_capsule, rotation=<rot 0 -90 0>}]
 base_Hair_01_r [{shapeType=Capsule, length=0.505, radius=0.505, shapeName=base_Hair_01_r_capsule, rotation=<rot 0 -90 0>}]
 dyn_Hair_01_l [{
   shapeType=Capsule
   length=1.9138
   radius=1.1978
   shapeName=dyn_Hair_01_l_capsule
   center=<1.4 -0.05 0.1>
   rotation=<rot -0.0283 -88.4807 -87.9632>
  }]
 dyn_Hair_01_m [{
   shapeType=Capsule
   length=1.6467
   radius=1.1273
   shapeName=dyn_Hair_01_m_capsule
   center=<1.1353 -0.0458 0.2663>
   rotation=<rot 0 -92.599 -89.9712>
  }]
 dyn_Hair_01_r [{
   shapeType=Capsule
   length=1.8384
   radius=1.5542
   shapeName=dyn_Hair_01_r_capsule
   center=<0.9915 -0.0311 0.1802>
   rotation=<rot 0.091 -87.5185 -88.0717>
  }]
 dyn_Hair_02_l [{shapeType=Capsule, length=2.3896, radius=1.1701, shapeName=dyn_Hair_02_l_capsule, center=<1.2 0.1 0.3>, rotation=<rot 0 -89.1079 -81.5316>}]
 dyn_Hair_02_m [{
   shapeType=Capsule
   length=1.729
   radius=1.0728
   shapeName=dyn_Hair_02_m_capsule
   center=<0.704 -0.2768 0.2707>
   rotation=<rot 0 -96.0351 -69.414>
  }]
 dyn_Hair_02_r [{
   shapeType=Capsule
   length=2.2875
   radius=1.4484
   shapeName=dyn_Hair_02_r_capsule
   center=<0.7132 -0.1276 0.454>
   rotation=<rot 0.7945 -96.3125 -78.3362>
  }]
 head [{shapeType=Capsule, length=0.3785, radius=7.8907, shapeName=head_capsule, center=<2.5444 2.2275 -0.2413>, rotation=<rot -90 0 0>}]
 neck_02 [
  {
   shapeType=Capsule
   length=4.444
   radius=2.5446
   shapeName=neck_02_capsule
   center=<0.5552 -4.8925 -0.057>
   rotation=<rot -0.0511 168.8944 -179.0359>
  }
  {
   shapeType=Capsule
   length=5.6068
   radius=2.5032
   shapeName=neck_02_capsule
   center=<0.2385 -3.5637 5.897>
   rotation=<rot 4.206 2.9953 35.5119>
  }
  {
   shapeType=Capsule
   length=4.9004
   radius=2.4656
   shapeName=neck_02_capsule
   center=<0.1807 -3.363 -5.9779>
   rotation=<rot -4.7668 3.8201 -38.834>
  }
 ]
 spine_05 [{shapeType=Capsule, length=21.8689, radius=11.0689, shapeName=spine_05_capsule, center=<-3.7914 0.3267 0>}]
}
constraints [
 {bone1Name=base_Hair_01_l, bone2Name=head, swing1LimitDegrees=1, swing2LimitDegrees=1, twistLimitDegrees=1}
 {bone1Name=dyn_Hair_01_l, bone2Name=base_Hair_01_l, swing1LimitDegrees=0, swing2LimitDegrees=0, twistLimitDegrees=0}
 {bone1Name=dyn_Hair_02_l, bone2Name=dyn_Hair_01_l, swing1LimitDegrees=5, swing2LimitDegrees=5, twistLimitDegrees=0}
 {bone1Name=base_Hair_01_m, bone2Name=head, swing1LimitDegrees=1, swing2LimitDegrees=1, twistLimitDegrees=1}
 {bone1Name=dyn_Hair_01_m, bone2Name=base_Hair_01_m, swing1LimitDegrees=0, swing2LimitDegrees=0, twistLimitDegrees=0}
 {bone1Name=dyn_Hair_02_m, bone2Name=dyn_Hair_01_m, swing1LimitDegrees=5, swing2LimitDegrees=5, twistLimitDegrees=0}
 {bone1Name=base_Hair_01_r, bone2Name=head, swing1LimitDegrees=1, swing2LimitDegrees=1, twistLimitDegrees=1}
 {bone1Name=dyn_Hair_01_r, bone2Name=base_Hair_01_r, swing1LimitDegrees=0, swing2LimitDegrees=0, twistLimitDegrees=0}
 {bone1Name=dyn_Hair_02_r, bone2Name=dyn_Hair_01_r, swing1LimitDegrees=5, swing2LimitDegrees=5, twistLimitDegrees=0}
]
